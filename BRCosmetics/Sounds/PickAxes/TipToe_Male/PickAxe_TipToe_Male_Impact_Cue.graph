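# Sound cue for the TipToe Male cosmetic pickaxe impact, using enemy-impact sound waves.
# A fixed short delay leads into an equally weighted random pick among six impact variations.
# Shared pickaxe impact attenuation, the Fort_Impacts sound class and the player melee weapon concurrency set govern playback, and the cue primes on load.
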SoundCue /BRCosmetics/Sounds/PickAxes/TipToe_Male/PickAxe_TipToe_Male_Impact_Cue
root=Delay_1
nodes [
 {id=Delay_1, children [Random_0], properties {delayMax=0.1, delayMin=0.1}}
 {
  id=Random_0
  children [WavePlayer_5, WavePlayer_6, WavePlayer_7, WavePlayer_8, WavePlayer_9, WavePlayer_11]
  properties {weights [1,1,1,1,1,1]}
 }
 {id=WavePlayer_11, properties {soundWaveAssetPtr=&/Pickaxe_TipToeMale_Enemy_Impact_06}}
 {id=WavePlayer_5, properties {soundWaveAssetPtr=&/Pickaxe_TipToeMale_Enemy_Impact_01}}
 {id=WavePlayer_6, properties {soundWaveAssetPtr=&/Pickaxe_TipToeMale_Enemy_Impact_02}}
 {id=WavePlayer_7, properties {soundWaveAssetPtr=&/Pickaxe_TipToeMale_Enemy_Impact_03}}
 {id=WavePlayer_8, properties {soundWaveAssetPtr=&/Pickaxe_TipToeMale_Enemy_Impact_04}}
 {id=WavePlayer_9, properties {soundWaveAssetPtr=&/Pickaxe_TipToeMale_Enemy_Impact_05}}
]
properties {
 attenuationSettings=/Game/Sounds/Attenuation/Fort_Pickaxe_Impact_Att
 audioPropertiesSheet=/BRCosmetics/Sounds/Settings/PropertySheets/Pickaxe/APS_Cosmetic_Pickaxe_Impact
 bPrimeOnLoad=true
 maxDistance=4300
 soundClassObject=&Fort_Impacts
 volumeMultiplier=0.675
 attenuationOverrides {
  bAttenuateWithLPF=true
  distanceAlgorithm=Logarithmic
  falloffDistance=800
  lPFRadiusMax=800
  lPFRadiusMin=400
  attenuationShapeExtents {x=300}
 }
 concurrencyOverrides {bLimitToOwner=true, maxCount=2, resolutionRule=StopOldest}
 concurrencySet [&Player_Melee_Weapon_Conc]
}
